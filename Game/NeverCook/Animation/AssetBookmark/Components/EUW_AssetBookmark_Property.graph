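# Editor utility widget that serves as a cosmetic item picker for a Multi Cam Viewer ID button.
# A single property view points at the widget itself in PreConstruct, exposing its own item variables for editing.
# On a property change, the parent's CosmeticType selects the emote, character, Sparks instrument or pickaxe variable.
# That item becomes the parent button's ItemDef, then the parent's Confirm Button is called.
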EditorUtilityWidgetBlueprint /Game/NeverCook/Animation/AssetBookmark/Components/EUW_AssetBookmark_Property
graphs [EventGraph {nodes [
   "On Property Changed (SinglePropertyView_65)"!() -> (OutputDelegate, then, PropertyName:Name=None) {is=title K2Node_ComponentBoundEvent}
   n1 = Get Parent() -> (Parent:EUW Multi Cam Viewer Id Button Object Reference)
   > Set ItemDef!(
    execute
    ItemDef:Item Definition Base Object Reference=Select!(
     Option 0:Item Definition Base Object Reference=Get EID() -> (EID:Athena Dance Item Definition Object Reference)
     Option 1:Item Definition Base Object Reference=Get CID() -> (CID:Athena Character Item Definition Object Reference)
     Option 2:Item Definition Base Object Reference=Get Sparks() -> (Sparks:Sparks Instrument Item Def Base Object Reference)
     Option 3:Item Definition Base Object Reference=Get Pickaxe() -> (Pickaxe:Athena Pickaxe Item Definition Object Reference)
     Index:int=Get CosmeticType!(self:EUW Multi Cam Viewer Id Button Object Reference=@n1) -> (CosmeticType:int=0)
    ) -> (ReturnValue:Item Definition Base Object Reference)
    self:EUW Multi Cam Viewer Id Button Object Reference=@n1
   ) -> (then, Output_Get:Item Definition Base Object Reference)
   > Confirm Button(self:EUW Multi Cam Viewer Id Button Object Reference=@n1)
   Event PreConstruct()
   > SetObject(self=Get SinglePropertyView_65() -> (SinglePropertyView_65:Single Property View Object Reference), NewObject=Self-Reference())
  ]}]
variables [
 CID {type=AthenaCharacterItemDefinition, default=/Game/Athena/Items/Cosmetics/Characters/CID_069_Athena_Commando_F_PinkBear}
 EID {type=AthenaDanceItemDefinition, default=/Game/ArtTools/Tech_Anim/TestTools/Emotes/EID_circleKickTechAnim}
 Parent {type=EUW_MultiCamViewer_IdButton_C, default=null}
 Pickaxe {type=AthenaPickaxeItemDefinition, default=null}
 Sparks {type=SparksInstrumentItemDefBase, default=null}
]
widgets {
 info {rootWidgetClass=&SinglePropertyView, widgetCount=1}
 widgets [SinglePropertyView_65 {widgetClassPath=&SinglePropertyView, widget=SinglePropertyView_65, properties {bAutoLoadAsset=false}, slot=~}]
}
